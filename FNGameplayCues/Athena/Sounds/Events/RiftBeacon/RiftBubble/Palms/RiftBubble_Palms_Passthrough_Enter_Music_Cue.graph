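# Sound cue for the music stinger of the Palms Rift Beacon bubble, named for passing through the bubble on entry.
# A mixer feeds a Fort_Gadgets sound class branch that randomly picks one of two equally weighted MeatPizza stingers.
# Its second, quieter branch overrides to the music packs sound class but has no child connected.
# Playback uses safe zone shield attenuation and allows one instance per owner, stopping the oldest.
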SoundCue /FNGameplayCues/Athena/Sounds/Events/RiftBeacon/RiftBubble/Palms/RiftBubble_Palms_Passthrough_Enter_Music_Cue
root=Mixer_0
nodes [
 {id=Mixer_0, children [SoundClass_0,SoundClass_1], properties {inputVolume [1,0.3]}}
 {id=Random_1, children [WavePlayer_0,WavePlayer_5], properties {weights [1,1]}}
 {id=SoundClass_0, children [Random_1], properties {soundClassOverride=&Fort_Gadgets}}
 {id=SoundClass_1, children [null], properties {soundClassOverride=/Game/Sounds/FortSoundClassesAndModes/Fort_Music_MusicPacks}}
 {id=WavePlayer_0, properties {soundWaveAssetPtr=&/MeatPizza_Music_Stinger_02}}
 {id=WavePlayer_5, properties {soundWaveAssetPtr=&/MeatPizza_Music_Stinger_01}}
]
properties {
 attenuationSettings=/Game/Sounds/Attenuation/Fort_SafeZone_Shield_Athena_Att
 bOverrideConcurrency=true
 maxDistance=4000
 volumeMultiplier=0.8
 concurrencyOverrides {bLimitToOwner=true, maxCount=1, resolutionRule=StopOldest}
}
